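# Physics asset for the male medium M_MED_Contaminate player body, covering only the hoodie, head and neck, apparently for secondary hood motion.
# Each dyn_ hoodie bone has capsules and is constrained to its small-capsule base_ counterpart with tight swing and twist limits, forming independent pairs rather than chains.
# Head is fully angle-locked to neck_02, whose capsules include a large head-named volume, likely as collision for the hood.
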PhysicsAsset /Game/Characters/Player/Male/Medium/Bodies/M_MED_Contaminate/Meshes/M_MED_Contaminate_Physics
bodies {
 base_B_hoodie01 [{shapeType=Capsule, length=0.505, radius=0.505, shapeName=base_B_hoodie01_capsule, rotation=<rot 0 -90 0>}]
 base_B_hoodie02 [{shapeType=Capsule, length=0.505, radius=0.505, shapeName=base_B_hoodie02_capsule, rotation=<rot 0 -90 0>}]
 base_B_hoodieIn05 [{shapeType=Capsule, length=0.505, radius=0.505, shapeName=base_B_hoodieIn05_capsule, rotation=<rot 0 -90 0>}]
 base_C_hoodieIn01 [{shapeType=Capsule, length=0.505, radius=0.505, shapeName=base_C_hoodieIn01_capsule}]
 base_L_hoodie01 [{shapeType=Capsule, length=1, radius=1, shapeName=base_L_hoodie01_capsule}]
 base_L_hoodie02 [{shapeType=Capsule, length=1, radius=1, shapeName=base_L_hoodie02_capsule}]
 base_L_hoodie03 [{shapeType=Capsule, length=1, radius=1, shapeName=base_L_hoodie03_capsule}]
 base_L_hoodieIn02 [{shapeType=Capsule, length=0.505, radius=0.505, shapeName=base_L_hoodieIn02_capsule}]
 base_L_hoodieIn03 [{shapeType=Capsule, length=0.505, radius=0.505, shapeName=base_L_hoodieIn03_capsule, rotation=<rot 0 -90 0>}]
 base_L_hoodieIn04 [{shapeType=Capsule, length=0.505, radius=0.505, shapeName=base_L_hoodieIn04_capsule, rotation=<rot 0 -90 0>}]
 base_R_hoodie01 [{shapeType=Capsule, length=1, radius=1, shapeName=base_R_hoodie01_capsule}]
 base_R_hoodie02 [{shapeType=Capsule, length=1, radius=1, shapeName=base_R_hoodie02_capsule}]
 base_R_hoodie03 [{shapeType=Capsule, length=1, radius=1, shapeName=base_R_hoodie03_capsule}]
 base_R_hoodieIn02 [{shapeType=Capsule, length=0.505, radius=0.505, shapeName=base_R_hoodieIn02_capsule}]
 base_R_hoodieIn03 [{shapeType=Capsule, length=0.505, radius=0.505, shapeName=base_R_hoodieIn03_capsule, rotation=<rot 0 -90 0>}]
 base_R_hoodieIn04 [{shapeType=Capsule, length=0.505, radius=0.505, shapeName=base_R_hoodieIn04_capsule, rotation=<rot 0 -90 0>}]
 dyn_B_hoodie01 [
  {shapeType=Capsule, length=3.5, radius=0.8, shapeName=dyn_B_hoodie01_capsule, center=<14.0864 0 0>, rotation=<rot 90 0 0>}
  {shapeType=Capsule, length=5, radius=1.3, shapeName=dyn_B_hoodie01_capsule, center=<9.6035 0 -9.1671>, rotation=<rot -44.9555 0 0>}
 ]
 dyn_B_hoodie02 [{shapeType=Capsule, length=5.4, radius=2.6, shapeName=dyn_B_hoodie02_capsule, center=<15.8272 0 0>, rotation=<rot 90 0 0>}]
 dyn_B_hoodieIn05 [{shapeType=Capsule, length=3.5, radius=1.5, shapeName=dyn_B_hoodieIn05_capsule, center=<-4.847 0 0>, rotation=<rot 0 0 90>}]
 dyn_C_hoodieIn01 [{shapeType=Capsule, length=5.2, radius=1.3, shapeName=dyn_C_hoodieIn01_capsule, center=<-4.4517 0 1.2286>, rotation=<rot 0 0 90>}]
 dyn_L_hoodie01 [{shapeType=Capsule, length=1, radius=1, shapeName=dyn_L_hoodie01_capsule, center=<4 0 0>, rotation=<rot 90 0 0>}]
 dyn_L_hoodie02 [{shapeType=Capsule, length=1, radius=1, shapeName=dyn_L_hoodie02_capsule, center=<4 0 0>, rotation=<rot 90 0 0>}]
 dyn_L_hoodie03 [{shapeType=Capsule, length=4, radius=1, shapeName=dyn_L_hoodie03_capsule, center=<3.9683 0.0547 0.5482>, rotation=<rot 83.4228 120.0708 -180>}]
 dyn_L_hoodieIn02 [{shapeType=Capsule, length=6.5, radius=1.4, shapeName=dyn_L_hoodieIn02_capsule, center=<-2.9029 0 0.1168>, rotation=<rot 0 0 90>}]
 dyn_L_hoodieIn03 [{shapeType=Capsule, length=4, radius=1, shapeName=dyn_L_hoodieIn03_capsule, center=<-2.7873 0 0>, rotation=<rot 0 -17.7187 90>}]
 dyn_L_hoodieIn04 [{
   shapeType=Capsule
   length=3.5
   radius=1.5
   shapeName=dyn_L_hoodieIn04_capsule
   center=<-3.8849 -0.1013 0.6274>
   rotation=<rot 1.4338 -14.9604 106.0826>
  }]
 dyn_R_hoodie01 [{shapeType=Capsule, length=1, radius=1, shapeName=dyn_R_hoodie01_capsule, center=<4 0 0>, rotation=<rot 90 0 0>}]
 dyn_R_hoodie02 [{shapeType=Capsule, length=1, radius=1, shapeName=dyn_R_hoodie02_capsule, center=<4 0 0>, rotation=<rot 90 0 0>}]
 dyn_R_hoodie03 [{shapeType=Capsule, length=4, radius=1, shapeName=dyn_R_hoodie03_capsule, center=<4 0 0>, rotation=<rot 90 90 35>}]
 dyn_R_hoodieIn02 [{shapeType=Capsule, length=4.6, radius=1.1, shapeName=dyn_R_hoodieIn02_capsule, center=<-2.9029 0 0>, rotation=<rot 0 0 90>}]
 dyn_R_hoodieIn03 [{shapeType=Capsule, length=3.5, radius=1.5, shapeName=dyn_R_hoodieIn03_capsule, center=<-2.7873 0 0>, rotation=<rot 0 14.4292 90>}]
 dyn_R_hoodieIn04 [{shapeType=Capsule, length=3.5, radius=1.5, shapeName=dyn_R_hoodieIn04_capsule, center=<-3.8492 -0.1013 0.6274>, rotation=<rot 0 10 80.6893>}]
 head [{shapeType=Capsule, length=4.6, radius=3.1, shapeName=head_capsule, center=<-4.3511 5.0381 0>, rotation=<rot 0 10 90>}]
 neck_02 [
  {shapeType=Capsule, length=5.2, radius=4.1, shapeName=neck_02_capsule, center=<3 0 0>, rotation=<rot 90 0 0>}
  {shapeType=Capsule, length=8.505, radius=9.605, shapeName=head_capsule, center=<13.6166 0.3189 0>, rotation=<rot 90 26.5651 103.2781>}
 ]
}
constraints [
 {bone1Name=dyn_R_hoodie03, bone2Name=base_R_hoodie03, swing1LimitDegrees=10, swing2LimitDegrees=20, twistLimitDegrees=1}
 {bone1Name=dyn_R_hoodie02, bone2Name=base_R_hoodie02, swing1LimitDegrees=10, swing2LimitDegrees=15, twistLimitDegrees=1}
 {bone1Name=dyn_L_hoodie03, bone2Name=base_L_hoodie03, swing1LimitDegrees=10, swing2LimitDegrees=15, twistLimitDegrees=1}
 {bone1Name=dyn_L_hoodie02, bone2Name=base_L_hoodie02, swing1LimitDegrees=10, swing2LimitDegrees=12, twistLimitDegrees=1}
 {bone1Name=dyn_R_hoodie01, bone2Name=base_R_hoodie01, swing1LimitDegrees=10, swing2LimitDegrees=20, twistLimitDegrees=1}
 {bone1Name=dyn_L_hoodie01, bone2Name=base_L_hoodie01, swing1LimitDegrees=10, swing2LimitDegrees=20, twistLimitDegrees=1}
 {bone1Name=dyn_B_hoodie02, bone2Name=base_B_hoodie02, swing1LimitDegrees=10, swing2LimitDegrees=5, twistLimitDegrees=1}
 {bone1Name=dyn_B_hoodie01, bone2Name=base_B_hoodie01, swing1LimitDegrees=12, swing2LimitDegrees=2, twistLimitDegrees=8}
 {bone1Name=dyn_C_hoodieIn01, bone2Name=base_C_hoodieIn01, swing1LimitDegrees=5, swing2LimitDegrees=20, twistLimitDegrees=20}
 {bone1Name=dyn_L_hoodieIn02, bone2Name=base_L_hoodieIn02, swing1LimitDegrees=5, swing2LimitDegrees=30, twistLimitDegrees=5}
 {bone1Name=dyn_R_hoodieIn02, bone2Name=base_R_hoodieIn02, swing1LimitDegrees=5, swing2LimitDegrees=20, twistLimitDegrees=5}
 {bone1Name=dyn_R_hoodieIn03, bone2Name=base_R_hoodieIn03, swing1LimitDegrees=5, swing2LimitDegrees=30, twistLimitDegrees=1}
 {bone1Name=dyn_R_hoodieIn04, bone2Name=base_R_hoodieIn04, swing1LimitDegrees=5, swing2LimitDegrees=30, twistLimitDegrees=1}
 {bone1Name=dyn_L_hoodieIn03, bone2Name=base_L_hoodieIn03, swing1LimitDegrees=5, swing2LimitDegrees=30, twistLimitDegrees=1}
 {bone1Name=dyn_L_hoodieIn04, bone2Name=base_L_hoodieIn04, swing1LimitDegrees=5, swing2LimitDegrees=30, twistLimitDegrees=1}
 {bone1Name=dyn_B_hoodieIn05, bone2Name=base_B_hoodieIn05, swing1LimitDegrees=5, swing2LimitDegrees=30, twistLimitDegrees=1}
 {
  bone1Name=head
  bone2Name=neck_02
  swing1LimitDegrees=45
  swing1Motion=Locked
  swing2LimitDegrees=45
  swing2Motion=Locked
  twistLimitDegrees=45
  twistMotion=Locked
 }
]
