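# Secondary-motion AnimBP for the Contaminate medium male player body, copying the parent mesh pose to animate hoodie, zippers, pauldron and pads.
# Cached stages chain pad constraints and deform Control Rig, skydive offsets, weighted zipper constraints with AnimDynamics, and RigidBody simulation.
# Update-driven movement-state and head-rotation alphas blend additive bone corrections that adjust hood and armor.
# The event graph also feeds material WPO cloth sim, raising wind and turbulence while skydiving.
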AnimBlueprint /Game/Characters/Player/Male/Medium/Bodies/M_MED_Contaminate/Meshes/M_MED_Contaminate_AnimBP
target_skeleton=&/M_MED_Contaminate_Skeleton
graphs [
 AnimGraph {nodes [
   n1 = Get RotHead_L() -> (RotHead_L:double)
   n2 = "Transform (Modify) Bone - Bone: dyn_C_hoodieIn01"!(
    ComponentPose=Local To Component(LocalPose=Use cached pose 'EndRB'())
    Alpha=@n1
    Translation="0.681,1.044,-0.202"
    Rotation="-3.125,12.689,8.617"
   ) -> (Pose) {
    rotationMode=BMM_Additive
    translationMode=BMM_Additive
    alphaBoolBlend {blendOption=Linear}
    boneToModify {boneName=dyn_C_hoodieIn01}
    rotation=<rot -3.125 12.689 8.617>
    translation=<0.681 1.044 -0.202>
   }
   n3 = "Transform (Modify) Bone - Bone: dyn_L_hoodieIn02"!(ComponentPose=@n2, Alpha=@n1, Translation="0,0,0", Rotation="10.356,-3.538,-1.955") -> (Pose) {
    rotationMode=BMM_Additive
    translationMode=BMM_Additive
    alphaBoolBlend {blendOption=Linear}
    boneToModify {boneName=dyn_L_hoodieIn02}
    rotation=<rot 10.356 -3.538 -1.955>
   }
   n4 = "Transform (Modify) Bone - Bone: dyn_L_hoodie01"!(ComponentPose=@n3, Alpha=@n1, Translation="0,0,0", Rotation="-6.043,7.665,13.126") -> (Pose) {
    rotationMode=BMM_Additive
    translationMode=BMM_Additive
    alphaBoolBlend {blendOption=Linear}
    boneToModify {boneName=dyn_L_hoodie01}
    rotation=<rot -6.043 7.665 13.126>
   }
   n5 = "Constraint - base_C_zipper01 (PPPPPPP)"!(
    ComponentPose=@n4
    Alpha=1
    ConstraintWeights_0=0.9715
    ConstraintWeights_1:float=0.0178
    ConstraintWeights_2:float=0.0076
    ConstraintWeights_3:float=0.0013
    ConstraintWeights_4:float=0.0012
    ConstraintWeights_5:float=0.0005
    ConstraintWeights_6:float=0.0002
   ) -> (Pose) {
    alphaBoolBlend {blendOption=Linear}
    boneToModify {boneName=base_C_zipper01}
    constraintSetup [
     {targetBone {boneName=spine_05}}
     {targetBone {boneName=dyn_R_hoodie01}}
     {targetBone {boneName=spine_04}}
     {targetBone {boneName=clavicle_r}}
     {targetBone {boneName=dyn_L_hoodie01}}
     {targetBone {boneName=clavicle_l}}
     {targetBone {boneName=dyn_R_hoodie02}}
    ]
    constraintWeights [1,1,1,1,1,1,1]
   }
   n6 = "Constraint - base_C_zipper02 (PPP)"!(ComponentPose=@n5, Alpha=1, ConstraintWeights_0=0.9834, ConstraintWeights_1:float=0.0083, ConstraintWeights_2:float=0.0083) -> (Pose) {
    alphaBoolBlend {blendOption=Linear}
    boneToModify {boneName=base_C_zipper02}
    constraintSetup [{targetBone {boneName=dyn_C_hoodieIn01}}, {targetBone {boneName=dyn_L_hoodieIn02}}, {targetBone {boneName=dyn_R_hoodieIn02}}]
    constraintWeights [1,1,1]
   }
   n7 = "Constraint - base_L_zipper01 (PPPPPPP)"!(
    ComponentPose=@n6
    Alpha=1
    ConstraintWeights_0=0.8416
    ConstraintWeights_1:float=0.1308
    ConstraintWeights_2:float=0.0211
    ConstraintWeights_3:float=0.005
    ConstraintWeights_4:float=0.0007
    ConstraintWeights_5:float=0.0006
    ConstraintWeights_6:float=0.0001
   ) -> (Pose) {
    alphaBoolBlend {blendOption=Linear}
    boneToModify {boneName=base_L_zipper01}
    constraintSetup [
     {targetBone {boneName=dyn_L_hoodie01}}
     {targetBone {boneName=dyn_L_hoodie02}}
     {targetBone {boneName=spine_05}}
     {targetBone {boneName=clavicle_l}}
     {targetBone {boneName=spine_04}}
     {targetBone {boneName=dyn_L_hoodie03}}
     {targetBone {boneName=upperarm_l}}
    ]
    constraintWeights [1,1,1,1,1,1,1]
   }
   n8 = "Constraint - base_R_zipper01 (PPPPPP)"!(
    ComponentPose=@n7
    Alpha=1
    ConstraintWeights_0=0.5995
    ConstraintWeights_1:float=0.3797
    ConstraintWeights_2:float=0.0163
    ConstraintWeights_3:float=0.0036
    ConstraintWeights_4:float=0.0005
    ConstraintWeights_5:float=0.0004
   ) -> (Pose) {
    alphaBoolBlend {blendOption=Linear}
    boneToModify {boneName=base_R_zipper01}
    constraintSetup [
     {targetBone {boneName=dyn_R_hoodie01}}
     {targetBone {boneName=dyn_R_hoodie02}}
     {targetBone {boneName=spine_05}}
     {targetBone {boneName=clavicle_r}}
     {targetBone {boneName=dyn_R_hoodie03}}
     {targetBone {boneName=spine_04}}
    ]
    constraintWeights [1,1,1,1,1,1]
   }
   n9 = Anim Dynamics - Bone: dyn_C_zipper01!(ComponentPose=@n8, Alpha=1, GravityOverride:Vector="0,0,-980") -> (Pose) {
    angularDampingOverride=1
    angularSpringConstant=20000
    bAngularSpring=true
    bGravityOverrideInSimSpace=true
    bOverrideAngularDamping=true
    bUseGravityOverride=true
    gravityScale=0.35
    simulationSpace=World
    alphaBoolBlend {blendOption=Linear}
    boundBone {boneName=dyn_C_zipper01}
    componentLinearAccScale=<1 1 1>
    componentLinearVelScale=<1 1 1>
    gravityOverride {z=-980}
    physicsBodyDefinitions [{
      sphereCollisionRadius=0
      boundBone {boneName=dyn_C_zipper01}
      boxExtents=<5 40 40>
      constraintSetup {angularLimitsMax=<5 10 -5>, angularLimitsMin=<-5 -10 -80>, angularTarget=<1 0 0>}
      localJointOffset=<4 0 0>
     }]
   }
   n10 = Anim Dynamics - Bone: dyn_C_zipper02!(ComponentPose=@n9, Alpha=1, GravityOverride:Vector="0,0,-980") -> (Pose) {
    angularDampingOverride=1
    angularSpringConstant=20000
    bAngularSpring=true
    bGravityOverrideInSimSpace=true
    bOverrideAngularDamping=true
    bUseGravityOverride=true
    gravityScale=0.35
    simulationSpace=World
    alphaBoolBlend {blendOption=Linear}
    boundBone {boneName=dyn_C_zipper02}
    componentLinearAccScale=<1 1 1>
    componentLinearVelScale=<1 1 1>
    gravityOverride {z=-980}
    physicsBodyDefinitions [{
      sphereCollisionRadius=0
      boundBone {boneName=dyn_C_zipper02}
      boxExtents=<5 40 40>
      constraintSetup {angularLimitsMax=<5 10 -25>, angularLimitsMin=<-5 -10 -80>, angularTarget=<1 0 0>}
      localJointOffset=<4 0 0>
     }]
   }
   n11 = Anim Dynamics - Bone: dyn_L_zipper01!(ComponentPose=@n10, Alpha=1, GravityOverride:Vector="0,0,-980") -> (Pose) {
    angularDampingOverride=1
    angularSpringConstant=20000
    bAngularSpring=true
    bGravityOverrideInSimSpace=true
    bOverrideAngularDamping=true
    bUseGravityOverride=true
    gravityScale=0.35
    simulationSpace=World
    alphaBoolBlend {blendOption=Linear}
    boundBone {boneName=dyn_L_zipper01}
    componentLinearAccScale=<1 1 1>
    componentLinearVelScale=<1 1 1>
    gravityOverride {z=-980}
    physicsBodyDefinitions [{
      sphereCollisionRadius=0
      boundBone {boneName=dyn_L_zipper01}
      boxExtents=<5 20 20>
      constraintSetup {angularLimitsMax=<5 10 -10>, angularLimitsMin=<-5 -10 -90>, angularTarget=<1 0 0>}
      localJointOffset=<4 0 0>
     }]
   }
   n12 = Anim Dynamics - Bone: dyn_R_zipper01!(ComponentPose=@n11, Alpha=1, GravityOverride:Vector="0,0,-980") -> (Pose) {
    angularDampingOverride=1
    angularSpringConstant=20000
    bAngularSpring=true
    bGravityOverrideInSimSpace=true
    bOverrideAngularDamping=true
    bUseGravityOverride=true
    gravityScale=0.35
    simulationSpace=World
    alphaBoolBlend {blendOption=Linear}
    boundBone {boneName=dyn_R_zipper01}
    componentLinearAccScale=<1 1 1>
    componentLinearVelScale=<1 1 1>
    gravityOverride {z=-980}
    physicsBodyDefinitions [{
      sphereCollisionRadius=0
      boundBone {boneName=dyn_R_zipper01}
      boxExtents=<5 20 20>
      constraintSetup {angularLimitsMax=<5 10 0>, angularLimitsMin=<-5 -10 -80>, angularTarget=<1 0 0>}
      localJointOffset=<4 0 0>
     }]
   }
   Output Pose(Result=Component To Local(ComponentPose=@n12)) {layerGroup=None}
   n13 = "Constraint - dyn_shoulderPad_r (PPPPPP)"!(
    ComponentPose=Local To Component(LocalPose=Copy Pose From Mesh() {bUseAttachedParent=true})
    Alpha=1
    ConstraintWeights_0=0.5
    ConstraintWeights_1:float=0.1
    ConstraintWeights_2:float=0.4
    ConstraintWeights_3:float=0
    ConstraintWeights_4:float=0
    ConstraintWeights_5:float=0
   ) -> (Pose) {
    alphaBoolBlend {blendOption=Linear}
    boneToModify {boneName=dyn_shoulderPad_r}
    constraintSetup [
     {targetBone {boneName=deform_upperarm_up_r}}
     {targetBone {boneName=clavicle_r}}
     {targetBone {boneName=upperarm_r}}
     {targetBone {boneName=upperarm_twist_01_r}}
     {targetBone {boneName=upperarm_twist_02_r}}
     {targetBone {boneName=clavicle_l}}
    ]
    constraintWeights [1,1,1,1,1,1]
   }
   n14 = Look At - Bone: dyn_shoulderPad_r(ComponentPose=@n13, Alpha=1) {
    alphaBoolBlend {blendOption=Linear}
    boneToModify {boneName=dyn_shoulderPad_r}
    lookAtLocation=<4.2057 -1.3361 -7.9845>
    lookAtTarget {boneReference {boneName=lowerarm_r}}
    lookAt_Axis {axis=<-1 0>}
    lookUp_Axis {axis {x=0.1, z=-0.9}}
   }
   n15 = "Constraint - dyn_kneePad_l (PP)"!(ComponentPose=@n14, Alpha=1, ConstraintWeights_0=0.4, ConstraintWeights_1:float=0.6) -> (Pose) {
    alphaBoolBlend {blendOption=Linear}
    boneToModify {boneName=dyn_kneePad_l}
    constraintSetup [{targetBone {boneName=thigh_l}}, {targetBone {boneName=calf_l}}]
    constraintWeights [1,1]
   }
   n16 = "Constraint - dyn_kneePad_r (PP)"!(ComponentPose=@n15, Alpha=1, ConstraintWeights_0=0.4, ConstraintWeights_1:float=0.6) -> (Pose) {
    alphaBoolBlend {blendOption=Linear}
    boneToModify {boneName=dyn_kneePad_r}
    constraintSetup [{targetBone {boneName=thigh_r}}, {targetBone {boneName=calf_r}}]
    constraintWeights [1,1]
   }
   n17 = Control Rig(Source=Component To Local(ComponentPose=@n16)) {
    alphaBoolBlend {blendOption=Linear}
    controlRigAssetReference {blueprintRigClass=&/M_MED_Contaminate_DeformRig_ControlRig.M_MED_Contaminate_DeformRig_ControlRig_C}
   }
   Save cached pose 'StartEnd'(Pose=@n17) {cacheName=StartEnd}
   n18 = "Transform (Modify) Bone - Bone: base_pauldron"!(
    ComponentPose=Local To Component(LocalPose=Use cached pose 'EndZippers'())
    Alpha=Get isCrouching() -> (isCrouching:double)
    Translation="0.5,-0.8,0"
    Rotation="2,0,10"
   ) -> (Pose) {
    rotationMode=BMM_Additive
    rotationSpace=BCS_BoneSpace
    translationMode=BMM_Additive
    alphaBoolBlend {blendOption=Linear}
    alphaScaleBiasClamp {bInterpResult=true}
    boneToModify {boneName=base_pauldron}
    rotation {pitch=2, roll=10}
    translation=<0.5 -0.8>
   }
   n19 = "Transform (Modify) Bone - Bone: base_pauldron"(ComponentPose=@n18, Alpha=Get iscrouchingSprinting() -> (iscrouchingSprinting:double), Translation="0,0,0", Rotation="0,20,10") {
    rotationMode=BMM_Additive
    rotationSpace=BCS_BoneSpace
    translationMode=BMM_Additive
    alphaBoolBlend {blendOption=Linear}
    alphaScaleBiasClamp {bInterpResult=true, clampMax=0.8}
    boneToModify {boneName=base_pauldron}
    rotation {roll=10, yaw=20}
   }
   n20 = Get HeadRot() -> (HeadRot:double)
   n21 = "Transform (Modify) Bone - Bone: base_B_hoodie01"(ComponentPose=@n19, Alpha=@n20, Translation="5,1,0", Rotation="0,50,0", Scale="1,1,1") {
    rotationMode=BMM_Additive
    rotationSpace=BCS_BoneSpace
    translationMode=BMM_Additive
    translationSpace=BCS_BoneSpace
    alphaBoolBlend {blendOption=Linear}
    boneToModify {boneName=base_B_hoodie01}
    rotation {yaw=50}
    translation=<5 1>
   }
   n22 = "Transform (Modify) Bone - Bone: base_R_hoodieIn04"(ComponentPose=@n21, Alpha=@n20, Translation="8,0,4", Rotation="0,0,0", Scale="1,1,1") {
    translationMode=BMM_Additive
    translationSpace=BCS_BoneSpace
    alphaBoolBlend {blendOption=Linear}
    boneToModify {boneName=base_R_hoodieIn04}
    translation {x=8, z=4}
   }
   n23 = "Transform (Modify) Bone - Bone: base_B_hoodieIn05"(ComponentPose=@n22, Alpha=@n20, Translation="8,0,4", Rotation="0,0,0", Scale="1,1,1") {
    translationMode=BMM_Additive
    translationSpace=BCS_BoneSpace
    alphaBoolBlend {blendOption=Linear}
    boneToModify {boneName=base_B_hoodieIn05}
    translation {x=8, z=4}
   }
   n24 = "Transform (Modify) Bone - Bone: base_L_hoodieIn04"(ComponentPose=@n23, Alpha=@n20, Translation="8,0,4", Rotation="0,0,0", Scale="1,1,1") {
    translationMode=BMM_Additive
    translationSpace=BCS_BoneSpace
    alphaBoolBlend {blendOption=Linear}
    boneToModify {boneName=base_L_hoodieIn04}
    translation {x=8, z=4}
   }
   n25 = Get linearAccScale() -> (linearAccScale:Vector)
   n26 = RigidBody!(ComponentPose=@n24, Alpha, OverrideWorldGravity:Vector="0,0,-980", ComponentLinearAccScale:Vector=@n25, ComponentLinearVelScale:Vector=@n25) -> (Pose) {
    bClampLinearTranslationLimitToRefPose=true
    bOverrideWorldGravity=true
    overridePhysicsAsset=&/M_MED_Contaminate_Physics
    simulationSpace=BaseBoneSpace
    alphaBoolBlend {blendOption=Linear}
    baseBoneRef {boneName=spine_05}
    componentLinearAccScale=<0.3 0.2 0.2>
    componentLinearVelScale=<0.3 0.2 0.2>
    overrideWorldGravity {z=-980}
    simSpaceSettings {worldAlpha=0.5}
   }
   Save cached pose 'EndRB'(Pose=Component To Local(ComponentPose=@n26)) {cacheName=EndRB}
   n27 = "Constraint - base_R_zipper02 (PPPPPPPP)"!(
    ComponentPose=Local To Component(LocalPose=Use cached pose 'EndSkyDive'())
    Alpha=1
    ConstraintWeights_0=0.8159
    ConstraintWeights_1:float=0.1544
    ConstraintWeights_2:float=0.0155
    ConstraintWeights_3:float=0.0087
    ConstraintWeights_4:float=0.0028
    ConstraintWeights_5:float=0.002
    ConstraintWeights_6:float=0.0004
    ConstraintWeights_7:float=0.0003
   ) -> (Pose) {
    alphaBoolBlend {blendOption=Linear}
    boneToModify {boneName=base_R_zipper02}
    constraintSetup [
     {targetBone {boneName=dyn_B_hoodie01}}
     {targetBone {boneName=dyn_B_hoodie04fix}}
     {targetBone {boneName=dyn_B_hoodie02}}
     {targetBone {boneName=dyn_R_hoodie03}}
     {targetBone {boneName=spine_05}}
     {targetBone {boneName=clavicle_r}}
     {targetBone {boneName=dyn_R_hoodie02}}
     {targetBone {boneName=upperarm_twist_01_r}}
    ]
    constraintWeights [1,1,1,1,1,1,1,1]
   }
   n28 = "Constraint - base_L_zipper02 (PPPPPPPP)"!(
    ComponentPose=@n27
    Alpha=1
    ConstraintWeights_0=0.6415
    ConstraintWeights_1:float=0.2624
    ConstraintWeights_2:float=0.0623
    ConstraintWeights_3:float=0.0289
    ConstraintWeights_4:float=0.0024
    ConstraintWeights_5:float=0.002
    ConstraintWeights_6:float=0.0004
    ConstraintWeights_7:float=0.0002
   ) -> (Pose) {
    alphaBoolBlend {blendOption=Linear}
    boneToModify {boneName=base_L_zipper02}
    constraintSetup [
     {targetBone {boneName=dyn_B_hoodie01}}
     {targetBone {boneName=dyn_B_hoodie04fix}}
     {targetBone {boneName=dyn_L_hoodie03}}
     {targetBone {boneName=dyn_B_hoodie02}}
     {targetBone {boneName=clavicle_l}}
     {targetBone {boneName=spine_05}}
     {targetBone {boneName=upperarm_twist_01_l}}
     {targetBone {boneName=dyn_L_hoodie02}}
    ]
    constraintWeights [1,1,1,1,1,1,1,1]
   }
   n29 = "Constraint - base_L_zipper03 (PPPPPPPP)"!(
    ComponentPose=@n28
    Alpha=1
    ConstraintWeights_0=0.3084
    ConstraintWeights_1:float=0.2676
    ConstraintWeights_2:float=0.2023
    ConstraintWeights_3:float=0.1423
    ConstraintWeights_4:float=0.0636
    ConstraintWeights_5:float=0.0105
    ConstraintWeights_6:float=0.0028
    ConstraintWeights_7:float=0.0025
   ) -> (Pose) {
    alphaBoolBlend {blendOption=Linear}
    boneToModify {boneName=base_L_zipper03}
    constraintSetup [
     {targetBone {boneName=spine_04}}
     {targetBone {boneName=spine_03}}
     {targetBone {boneName=spine_02}}
     {targetBone {boneName=spine_01}}
     {targetBone {boneName=pelvis}}
     {targetBone {boneName=clavicle_l}}
     {targetBone {boneName=upperarm_l}}
     {targetBone {boneName=deform_groin_l}}
    ]
    constraintWeights [1,1,1,1,1,1,1,1]
   }
   n30 = "Constraint - base_R_zipper03 (PPPPPPPP)"!(
    ComponentPose=@n29
    Alpha=1
    ConstraintWeights_0=0.3786
    ConstraintWeights_1:float=0.2592
    ConstraintWeights_2:float=0.1796
    ConstraintWeights_3:float=0.1126
    ConstraintWeights_4:float=0.0474
    ConstraintWeights_5:float=0.0163
    ConstraintWeights_6:float=0.004
    ConstraintWeights_7:float=0.0022
   ) -> (Pose) {
    alphaBoolBlend {blendOption=Linear}
    boneToModify {boneName=base_R_zipper03}
    constraintSetup [
     {targetBone {boneName=spine_04}}
     {targetBone {boneName=spine_03}}
     {targetBone {boneName=spine_02}}
     {targetBone {boneName=spine_01}}
     {targetBone {boneName=pelvis}}
     {targetBone {boneName=clavicle_r}}
     {targetBone {boneName=upperarm_r}}
     {targetBone {boneName=upperarm_twist_01_r}}
    ]
    constraintWeights [1,1,1,1,1,1,1,1]
   }
   n31 = "Constraint - base_L_zipper04 (PPPPPPPP)"!(
    ComponentPose=@n30
    Alpha=1
    ConstraintWeights_0=0.5011
    ConstraintWeights_1:float=0.4536
    ConstraintWeights_2:float=0.0231
    ConstraintWeights_3:float=0.0106
    ConstraintWeights_4:float=0.0046
    ConstraintWeights_5:float=0.0025
    ConstraintWeights_6:float=0.0024
    ConstraintWeights_7:float=0.0022
   ) -> (Pose) {
    alphaBoolBlend {blendOption=Linear}
    boneToModify {boneName=base_L_zipper04}
    constraintSetup [
     {targetBone {boneName=lowerarm_l}}
     {targetBone {boneName=lowerarm_twist_02_l}}
     {targetBone {boneName=lowerarm_twist_01_l}}
     {targetBone {boneName=upperarm_l}}
     {targetBone {boneName=hand_l}}
     {targetBone {boneName=thumb_01_l}}
     {targetBone {boneName=middle_metacarpal_l}}
     {targetBone {boneName=index_metacarpal_l}}
    ]
    constraintWeights [1,1,1,1,1,1,1,1]
   }
   n32 = "Constraint - base_R_zipper04 (PPPPPPPP)"!(
    ComponentPose=@n31
    Alpha=1
    ConstraintWeights_0=0.5011
    ConstraintWeights_1:float=0.4536
    ConstraintWeights_2:float=0.0231
    ConstraintWeights_3:float=0.0106
    ConstraintWeights_4:float=0.0046
    ConstraintWeights_5:float=0.0025
    ConstraintWeights_6:float=0.0024
    ConstraintWeights_7:float=0.0022
   ) -> (Pose) {
    alphaBoolBlend {blendOption=Linear}
    boneToModify {boneName=base_R_zipper04}
    constraintSetup [
     {targetBone {boneName=lowerarm_r}}
     {targetBone {boneName=lowerarm_twist_02_r}}
     {targetBone {boneName=lowerarm_twist_01_r}}
     {targetBone {boneName=upperarm_r}}
     {targetBone {boneName=hand_r}}
     {targetBone {boneName=thumb_01_r}}
     {targetBone {boneName=middle_metacarpal_r}}
     {targetBone {boneName=index_metacarpal_r}}
    ]
    constraintWeights [1,1,1,1,1,1,1,1]
   }
   n33 = Anim Dynamics - Bone: dyn_R_zipper02!(ComponentPose=@n32, Alpha=1, GravityOverride:Vector="0,0,-980") -> (Pose) {
    angularDampingOverride=1
    angularSpringConstant=20000
    bAngularSpring=true
    bGravityOverrideInSimSpace=true
    bOverrideAngularDamping=true
    bUseGravityOverride=true
    gravityScale=0.35
    simulationSpace=World
    alphaBoolBlend {blendOption=Linear}
    boundBone {boneName=dyn_R_zipper02}
    componentLinearAccScale=<1 1 1>
    componentLinearVelScale=<1 1 1>
    gravityOverride {z=-980}
    physicsBodyDefinitions [{
      sphereCollisionRadius=0
      boundBone {boneName=dyn_R_zipper02}
      boxExtents=<5 40 40>
      constraintSetup {angularLimitsMax=<5 10 5>, angularLimitsMin=<-5 -10 -70>, angularTarget=<1 0 0>}
      localJointOffset=<4 0 0>
     }]
   }
   n34 = Anim Dynamics - Bone: dyn_L_zipper02!(ComponentPose=@n33, Alpha=1, GravityOverride:Vector="0,0,-980") -> (Pose) {
    angularDampingOverride=1
    angularSpringConstant=20000
    bAngularSpring=true
    bGravityOverrideInSimSpace=true
    bOverrideAngularDamping=true
    bUseGravityOverride=true
    gravityScale=0.35
    simulationSpace=World
    alphaBoolBlend {blendOption=Linear}
    boundBone {boneName=dyn_L_zipper02}
    componentLinearAccScale=<1 1 1>
    componentLinearVelScale=<1 1 1>
    gravityOverride {z=-980}
    physicsBodyDefinitions [{
      sphereCollisionRadius=0
      boundBone {boneName=dyn_L_zipper02}
      boxExtents=<5 40 40>
      constraintSetup {angularLimitsMax=<5 10 5>, angularLimitsMin=<-5 -10 -80>, angularTarget=<1 0 0>}
      localJointOffset=<4 0 0>
     }]
   }
   n35 = Anim Dynamics - Bone: dyn_L_zipper03!(ComponentPose=@n34, Alpha=1, GravityOverride:Vector="0,0,-980") -> (Pose) {
    angularDampingOverride=1
    angularSpringConstant=20000
    bAngularSpring=true
    bGravityOverrideInSimSpace=true
    bOverrideAngularDamping=true
    bUseGravityOverride=true
    gravityScale=0.35
    simulationSpace=World
    alphaBoolBlend {blendOption=Linear}
    boundBone {boneName=dyn_L_zipper03}
    componentLinearAccScale=<1 1 1>
    componentLinearVelScale=<1 1 1>
    gravityOverride {z=-980}
    physicsBodyDefinitions [{
      sphereCollisionRadius=0
      boundBone {boneName=dyn_L_zipper03}
      boxExtents=<5 40 40>
      constraintSetup {angularLimitsMax=<5 20 0>, angularLimitsMin=<-5 -20 -80>, angularTarget=<1 0 0>}
      localJointOffset=<4 0 0>
     }]
   }
   n36 = Anim Dynamics - Bone: dyn_R_zipper03!(ComponentPose=@n35, Alpha=1, GravityOverride:Vector="0,0,-980") -> (Pose) {
    angularDampingOverride=1
    angularSpringConstant=20000
    bAngularSpring=true
    bGravityOverrideInSimSpace=true
    bOverrideAngularDamping=true
    bUseGravityOverride=true
    gravityScale=0.35
    simulationSpace=World
    alphaBoolBlend {blendOption=Linear}
    boundBone {boneName=dyn_R_zipper03}
    componentAppliedLinearAccClamp {z=1000000}
    componentLinearAccScale=<1 1 10>
    componentLinearVelScale=<1 1 10>
    externalForce {z=1}
    gravityOverride {z=-980}
    physicsBodyDefinitions [{
      sphereCollisionRadius=0
      boundBone {boneName=dyn_R_zipper03}
      boxExtents=<5 40 40>
      constraintSetup {angularLimitsMax=<5 20 0>, angularLimitsMin=<-5 -20 -80>, angularTarget=<1 0 0>}
      localJointOffset=<4 0 0>
     }]
   }
   n37 = Anim Dynamics - Bone: dyn_L_zipper04!(ComponentPose=@n36, Alpha=1, GravityOverride:Vector="0,0,-980") -> (Pose) {
    angularDampingOverride=1
    angularSpringConstant=20000
    bAngularSpring=true
    bGravityOverrideInSimSpace=true
    bOverrideAngularDamping=true
    bUseGravityOverride=true
    gravityScale=0.35
    simulationSpace=World
    alphaBoolBlend {blendOption=Linear}
    boundBone {boneName=dyn_L_zipper04}
    componentLinearAccScale=<1 1 1>
    componentLinearVelScale=<1 1 1>
    gravityOverride {z=-980}
    physicsBodyDefinitions [{
      sphereCollisionRadius=0
      boundBone {boneName=dyn_L_zipper04}
      boxExtents=<5 40 40>
      constraintSetup {angularLimitsMax=<5 10 5>, angularLimitsMin=<-5 -10 -80>, angularTarget=<1 0 0>}
      localJointOffset=<4 0 0>
     }]
   }
   n38 = Anim Dynamics - Bone: dyn_R_zipper04!(ComponentPose=@n37, Alpha=1, GravityOverride:Vector="0,0,-980") -> (Pose) {
    angularDampingOverride=1
    angularSpringConstant=20000
    bAngularSpring=true
    bGravityOverrideInSimSpace=true
    bOverrideAngularDamping=true
    bUseGravityOverride=true
    gravityScale=0.35
    simulationSpace=World
    alphaBoolBlend {blendOption=Linear}
    boundBone {boneName=dyn_R_zipper04}
    componentLinearAccScale=<1 1 1>
    componentLinearVelScale=<1 1 1>
    gravityOverride {z=-980}
    physicsBodyDefinitions [{
      sphereCollisionRadius=0
      boundBone {boneName=dyn_R_zipper04}
      boxExtents=<5 40 40>
      constraintSetup {angularLimitsMax=<5 10 5>, angularLimitsMin=<-5 -10 -80>, angularTarget=<1 0 0>}
      localJointOffset=<4 0 0>
     }]
   }
   n39 = "Constraint - dyn_B_hoodie01 (P)"(ComponentPose=@n38, Alpha=1, ConstraintWeights_0=1) {
    alphaBoolBlend {blendOption=Linear}
    boneToModify {boneName=dyn_B_hoodie01}
    constraintSetup [{targetBone {boneName=neck_01}}]
    constraintWeights [1]
   }
   Save cached pose 'EndZippers'(Pose=Component To Local(ComponentPose=@n39)) {cacheName=EndZippers}
   n40 = Get isSkydiveFloating() -> (isSkydiveFloating:double)
   n41 = "Transform (Modify) Bone - Bone: base_pauldron"(ComponentPose=Local To Component(LocalPose=Use cached pose 'StartEnd'()), Alpha=@n40, Translation="2.5,0,0.5", Rotation="0,0,0") {
    rotationMode=BMM_Additive
    rotationSpace=BCS_BoneSpace
    translationMode=BMM_Additive
    alphaBoolBlend {blendOption=Linear}
    alphaScaleBiasClamp {bInterpResult=true}
    boneToModify {boneName=base_pauldron}
    translation {x=2.5, z=0.5}
   }
   n42 = "Transform (Modify) Bone - Bone: dyn_B_hoodie01"!(ComponentPose=@n41, Alpha=@n40, Translation="-0.161,-0.705,3.745", Rotation="0,0,0") -> (Pose) {
    rotationMode=BMM_Additive
    rotationSpace=BCS_BoneSpace
    translationMode=BMM_Additive
    alphaBoolBlend {blendOption=Linear}
    alphaScaleBiasClamp {bInterpResult=true}
    boneToModify {boneName=dyn_B_hoodie01}
    translation=<-0.161 -0.705 3.745>
   }
   n43 = "Transform (Modify) Bone - Bone: dyn_shoulderPad_r"!(ComponentPose=@n42, Alpha=Get isSkydiving() -> (isSkydiving:double), Translation="-1.480,0.211,1.423", Rotation="-9.911,14.177,9.672") -> (Pose) {
    rotationMode=BMM_Additive
    rotationSpace=BCS_BoneSpace
    translationMode=BMM_Additive
    alphaBoolBlend {blendOption=Linear}
    alphaScaleBiasClamp {bInterpResult=true}
    boneToModify {boneName=dyn_shoulderPad_r}
    rotation=<rot -9.911 14.177 9.672>
    translation=<-1.48 0.211 1.423>
   }
   n44 = Get isParachuteOpen() -> (isParachuteOpen:double)
   n45 = "Transform (Modify) Bone - Bone: dyn_shoulderPad_r"(ComponentPose=@n43, Alpha=@n44, Translation="0.103,0.504,0.386", Rotation="4.772,1.731,15.071") {
    rotationMode=BMM_Additive
    rotationSpace=BCS_BoneSpace
    translationMode=BMM_Additive
    alphaBoolBlend {blendOption=Linear}
    alphaScaleBiasClamp {bInterpResult=true}
    boneToModify {boneName=dyn_shoulderPad_r}
    rotation=<rot 4.772 1.731 15.071>
    translation=<0.103 0.504 0.386>
   }
   n46 = "Transform (Modify) Bone - Bone: base_pauldron"(ComponentPose=@n45, Alpha=@n44, Translation="2,0,0", Rotation="-7,0,0") {
    rotationMode=BMM_Additive
    rotationSpace=BCS_BoneSpace
    translationMode=BMM_Additive
    translationSpace=BCS_BoneSpace
    alphaBoolBlend {blendOption=Linear}
    alphaScaleBiasClamp {bInterpResult=true}
    boneToModify {boneName=base_pauldron}
    rotation {pitch=-7}
    translation {x=2}
   }
   n47 = Get isJumpingSkydive() -> (isJumpingSkydive:double)
   n48 = "Transform (Modify) Bone - Bone: base_B_hoodie01"!(ComponentPose=@n46, Alpha=@n47, Translation="0,0,0", Rotation="0,30,0") -> (Pose) {
    rotationMode=BMM_Additive
    rotationSpace=BCS_BoneSpace
    translationMode=BMM_Additive
    translationSpace=BCS_BoneSpace
    alphaBoolBlend {blendOption=Linear}
    alphaScaleBiasClamp {bInterpResult=true}
    boneToModify {boneName=base_B_hoodie01}
    rotation {yaw=30}
   }
   n49 = "Transform (Modify) Bone - Bone: base_B_hoodie02"!(ComponentPose=@n48, Alpha=@n47, Translation="0,0,0", Rotation="0,30,0") -> (Pose) {
    rotationMode=BMM_Additive
    rotationSpace=BCS_BoneSpace
    translationMode=BMM_Additive
    translationSpace=BCS_BoneSpace
    alphaBoolBlend {blendOption=Linear}
    alphaScaleBiasClamp {bInterpResult=true}
    boneToModify {boneName=base_B_hoodie02}
    rotation {yaw=30}
   }
   Save cached pose 'EndSkyDive'(Pose=Blend Poses by bool(
     BlendPose_0=Component To Local(ComponentPose=@n49)
     BlendPose_1=Use cached pose 'StartEnd'()
     BlendTime_0=1
     BlendTime_1=0.2
     bActiveValue=Get bIsSkyDiving()
    ) {blendType=Linear}) {cacheName=EndSkyDive}
   Reroute Node(InputPin:Wildcard) -> (OutputPin:Wildcard)
  ]}
 EventGraph {nodes [
   n1 = Event BlueprintUpdateAnimation()
   > n2 = Sequence!(execute) -> (then_0, then_1, then_2:Exec, then_3:Exec)
   > n3 = Gravity Override State Function DEPRECATED(CharacterPartAnimInstance=Self-Reference(), GameTimeInSeconds=GetGameTimeInSeconds(), GravityOverrideParams=Get Gravity Override Params())
   > Set GravityOverride(GravityOverride=@n3)
   Set WPO Globals Function(
    execute=@n2.then_1
    Material=Get Mat Instance Dynamic()
    Overlay Material=Get Overlay Mat Instance Dynamic()
    Skeletal Mesh=Get SkelMesh()
    "Local Noise Transform  Location"=Get Local Noise Transform Location() -> (Local Noise Transform Location:Vector)
    Local Noise Offset=Get Local Noise Offset()
   )
   > Set WPO Regional Function!(
    execute
    Material:Material Instance Dynamic Object Reference=Get Mat Instance Dynamic()
    Overlay Material:Material Instance Dynamic Object Reference=Get Overlay Mat Instance Dynamic()
    Skeletal Mesh:Skeletal Mesh Component Object Reference=Get SkelMesh()
    Velocity Transform:Name=foot_r
    Velocity Interp Speed:double=Get Velocity Interp Speed()
    Collision Transform:Name=calf_r
    Collision Transform Axis:EAxis Enum
    Invert Collision Transform Axis:bool=false
    Delta Time X:double=@n1.DeltaTimeX
    Previous Velocity Transform Location:"Vector (by ref)"=Get R1 Velocity Transform Location()
    Previous Velocity Transform Velocity:"Vector (by ref)"=Get R1 Velocity Transform Velocity()
    Region:int=1
   ) -> (then)
   > Set WPO Regional Function(
    Material=Get Mat Instance Dynamic()
    Overlay Material=Get Overlay Mat Instance Dynamic()
    Skeletal Mesh=Get SkelMesh()
    Velocity Transform=foot_l
    Velocity Interp Speed=Get Velocity Interp Speed()
    Collision Transform=calf_l
    Invert Collision Transform Axis=true
    Delta Time X=@n1.DeltaTimeX
    Previous Velocity Transform Location=Get R2 Velocity Transform Location()
    Previous Velocity Transform Velocity=Get R2 Velocity Transform Velocity()
    Region=2
   )
   > n4 = Branch(Condition=Get bIsSkyDiving())
   > Set WPO Pho Wind Intensity(WPO Pho Wind Intensity=0.8)
   > n5 = Set WPO Pho Turbulence(WPO Pho Turbulence:double=0.8) -> (Output_Get:double)
   SetScalarParameterValue(
    execute=[@n5.then, @n7.then]
    self=[Get Overlay Mat Instance Dynamic(), Get Mat Instance Dynamic()]
    ParameterName=ClothSim_WindStrength
    Value=Get WPO Pho Wind Intensity()
   )
   n6 = Get WPO Pho Turbulence() -> (WPO Pho Turbulence:double)
   > SetScalarParameterValue(self=[Get Overlay Mat Instance Dynamic(), Get Mat Instance Dynamic()], ParameterName=ClothSim_R1_TurbulenceStrength, Value=@n6)
   > SetScalarParameterValue(self=[Get Mat Instance Dynamic(), Get Overlay Mat Instance Dynamic()], ParameterName=ClothSim_R2_TurbulenceStrength, Value=@n6)
   Set WPO Pho Wind Intensity(execute=@n4.else, WPO Pho Wind Intensity=0.1)
   > n7 = Set WPO Pho Turbulence(WPO Pho Turbulence=0.2)
   Set isSkydiveFloating(execute=@n2.then_2, isSkydiveFloating:double="To Float (Boolean)"(InBool=Get bIsSkydiveFloating())) -> (Output_Get:double)
   n8 = Get bIsSkyDiving()
   > Set isSkydiving(isSkydiving:double="To Float (Boolean)"(InBool=@n8)) -> (Output_Get:double)
   > Set isParachuteOpen(isParachuteOpen:double="To Float (Boolean)"(InBool=Get bIsParachuteOpened())) -> (Output_Get:double)
   > Set iscrouchingSprinting(iscrouchingSprinting:double="To Float (Boolean)"(InBool=Get bIsCrouchSprinting())) -> (Output_Get:double)
   > Set isCrouching(isCrouching:double="To Float (Boolean)"(InBool=Get bIsCrouching())) -> (Output_Get:double)
   n9 = AND Boolean(A=Get bIsJumping(), B=@n8)
   > Set isJumpingSkydive(isJumpingSkydive:double="To Float (Boolean)"(InBool=@n9)) -> (Output_Get:double)
   > Set isJumpingSkydive_bool(isJumpingSkydive_bool:bool=@n9) -> (Output_Get:bool)
   > Set RotHead_L(RotHead_L:double=GetAlphaFromReferenceTransformDelta(SkeletalMeshComponent=Get SkelMesh(), BoneName=spine_05, BaseName=head, DeltaComponentType=EulerX, InputRange=(X=0,Y=-22), TwistAxis="1,0,0")) -> (Output_Get:double)
   > n10 = Branch(Condition=@n8)
   > Set linearAccScale(linearAccScale:Vector="0.15,0.05,0.05") -> (Output_Get:Vector)
   Set linearAccScale(execute=@n10.else, linearAccScale="0.3,0.2,0.2")
   Set HeadRot(
    execute=@n2.then_3
    HeadRot:double=GetAlphaFromReferenceTransformDelta(SkeletalMeshComponent=Get SkelMesh(), BoneName=spine_04, BaseName=head, InputRange=(X=25,Y=80), TwistAxis="0,0,1")
   ) -> (Output_Get:double)
   Event OnCharacterCustomizationLoaded()
   > n11 = Set SkelMesh(SkelMesh=GetPartSkeletalMeshComponent())
   > n12 = Get Dynamic Material Instance(SkeletalMeshComponent=@n11)
   > Set Mat Instance Dynamic(Mat Instance Dynamic=@n12)
   > n13 = GetOrCreateDynamicOverlayMaterial(MeshComponent=@n11, MaterialName=None)
   > Set Overlay Mat Instance Dynamic(Overlay Mat Instance Dynamic=@n13)
  ]}
]
variables [
 BodyMesh {type=SkeletalMeshComponent}
 Gravity Override Params {type=GravityOverrideParamsStruct}
 GravityOverride {type=Vector}
 head rot {type=Rotator}
 head Rot Back {type=number}
 HeadRot {type=number}
 isCrouching {type=number}
 iscrouchingSprinting {type=number}
 isJumpingSkydive {type=number}
 isJumpingSkydive_bool {type=boolean}
 isParachuteOpen {type=number}
 isSkydiveFloating {type=number}
 isSkydiving {type=number}
 linearAccScale {type=Vector}
 Local Noise Offset {type=Vector}
 Local Noise Transform Location {type=Vector}
 Mat Instance Dynamic {type=MaterialInstanceDynamic}
 Overlay Mat Instance Dynamic {type=MaterialInstanceDynamic}
 R1 Velocity Transform Location {type=Vector}
 R1 Velocity Transform Velocity {type=Vector}
 R2 Velocity Transform Location {type=Vector}
 R2 Velocity Transform Velocity {type=Vector}
 RotHead_L {type=number}
 SkelMesh {type=SkeletalMeshComponent}
 Velocity Interp Speed {type=number}
 WPO Pho Turbulence {type=number}
 WPO Pho Wind Intensity {type=number}
 X percentage {type=number}
 Y percentage {type=number}
 Z percentage {type=number}
]
